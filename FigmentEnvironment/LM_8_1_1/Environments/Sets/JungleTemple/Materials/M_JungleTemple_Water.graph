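# Water material for the Jungle Temple environment, lit per pixel with screen-space reflections and world-space normals.
# By default inputs, the CubeFX static switch selects between ShallowWaterFunction_JungleTemple and an alternate attribute branch fed by an unconnected reroute.
# In that branch, UseEdgeMaskTransition adds radial-gradient edge color to emissive, depth-fades opacity and sets constant refraction.
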Material /FigmentEnvironment/LM_8_1_1/Environments/Sets/JungleTemple/Materials/M_JungleTemple_Water
graph {
 properties {
  bComputeFogPerPixel=true
  bScreenSpaceReflections=true
  bTangentSpaceNormal=false
  bUseHQForwardReflections=true
  bUseMaterialAttributes=true
  bUsePlanarForwardReflections=true
  pixelDepthOffsetMode=PDOM_Legacy
  refractionMethod=RM_PixelNormalOffset
  translucencyLightingMode=TLM_SurfacePerPixelLighting
  translucencyPass=MTP_BeforeDOF
  usedShadingModels=Default Lit
 }
 nodes [
  n1 = Reroute()
  n2 = GetMaterialAttributes(None=@n1) {attributeGetTypes [#1]}
  n3 = TextureSample() {texture=/FigmentEnvironment/LM_8_1_1/UI/Frontend/Store/Textures/T_RadGradient}
  n4 = Add(
   A=Multiply(A=VectorParameter() {parameterName=DetailColor_Edge, defaultValue=<rgba 0.4409 0.1856 2 0>}.RGB, B=OneMinus(None=@n3.G))
   B=@n2.Emissive Color
  )
  n5 = DepthFade(Opacity=ScalarParameter() {defaultValue=60, parameterName=EdgeOpactiy}, FadeDistance=ScalarParameter() {parameterName=FadeDistance})
  n6 = Constant() {r=1.25}
  StaticSwitchParameter(
   True=StaticSwitchParameter(
    True=SetMaterialAttributes!(MaterialAttributes=@n2.MaterialAttributes, Emissive Color=@n4, Opacity=@n5, Refraction=@n6) {attributeSetTypes [#1, #2, #3]}
    False=@n1
   ) {parameterName=UseEdgeMaskTransition}
   False=MaterialFunctionCall!(
    Height Offset Amount
    Sample Water Normal
    Sample Water Height Map
    Include water interaction modifications=StaticBool()
    Ground texture mask
   ) {materialFunction=/FigmentEnvironment/LM_8_1_1/Environments/Sets/JungleTemple/MaterialFunctions/ShallowWaterFunction_JungleTemple}.MA Result
  ) {parameterName=CubeFX}
 ]
}
